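# Niagara system for the Icy Scarf cosmetic companion emote, producing swirling snowflakes with brief mesh swish effects.
# Vortex emitters swirl flakes inside a kill-bounded sphere, while SL emitters place them on a rotating sphere shell, each with GPU and CPU "Low" twins.
# Two timed bursts use spawn groups to vary lifetime, drag, vortex force and renderer visibility tag, with time-of-day tinting.
# Two WaterSwishes emitters add short-lived, counter-rotating meshes early and late in the sequence.
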NiagaraSystem /CosmeticCompanions/FX/Builtins/IcyScarf/Emote/FX/NS_Emote_IcyScarf_SnowFlakes
dependencies {
 usedDataInterfaces [
  &NiagaraDataInterfaceCurve
  &NiagaraDataInterfaceArrayFloat
  &NiagaraDataInterfaceArrayInt32
  &NiagaraDataInterfaceVectorCurve
 ]
 usedDynamicInputs []
 usedModules [&SystemState]
}
emitters {
 Snowflakes_SL_GPU {
  simTarget=GPUComputeSim
  emitterUpdateScript [
   &EmitterState(Life Cycle Mode=::NewEnumerator1, Loop Behavior=::NewEnumerator1, Loop Delay=0.5, Loop Duration=9, UseLoopDelay=true)
   &SpawnBurst_Instantaneous(Spawn Count=75, Use Spawn Probability=false, UseLoopCountLimit=false)
   &SpawnBurst_Instantaneous(Spawn Count=70, Spawn Group=1, Spawn Time=6.2, Use Spawn Probability=false, UseLoopCountLimit=false) {module=SpawnBurst_Instantaneous001}
  ]
  particleSpawnScript [
   &InitializeParticle(
    Color=&FNTimeOfDayColor(
     Night Color=&RandomRangeLinearColor(Minimum=<rgba 0.3449 0.6586 1 0.8>, Maximum=<rgba 0.4305 0.5338 0.6354 0.8>)
     Morning Color=&RandomRangeLinearColor(Minimum=<rgba 0.3449 0.6586 1 0.8>, Maximum=<rgba 0.4305 0.5338 0.6354 0.8>)
     Day Color=&RandomRangeLinearColor(Minimum=<rgba 0.4138 0.7904 1.2 1>, Maximum=<rgba 0.6775 0.8401 1 1>)
     Evening Color=&RandomRangeLinearColor(Minimum=<rgba 0.3449 0.6586 1 0.8>, Maximum=<rgba 0.4305 0.5338 0.6354 0.8>)
    )
    Lifetime=&MergeHelper_FloatSwitch(
     0=&RandomRangeFloat(Minimum=5.5, Maximum=6.5)
     1=&RandomRangeFloat(Minimum=1.8, Maximum=2.2)
     2=?
     3=?
     Index=&SpawnGroup()
     Num=::NewEnumerator0
    )
    Mesh Scale Mode=::NewEnumerator1
    Mesh Uniform Scale Max=0.06
    Mesh Uniform Scale Min=0.015
    Sprite Size Mode=::NewEnumerator1
    Uniform Sprite Size Max=9
    Uniform Sprite Size Min=1
    UsePositionOffset=true
    Write Material Random=true
   )
   &AddVelocity(Velocity=<0 0 0>)
   /SetVariables(Particles.VisibilityTag=&SpawnGroup()) {isSetParameters=true, module=SetVariables_#1}
   &CameraOffset(Camera Offset Amount=-5)
  ]
  particleUpdateScript [
   &ParticleState()
   &UpdateMeshOrientation(
    Rotation Rate=&FloatFromCurve(FloatCurve=DataInterface() {propertyValues {Curve {keys [<at 0 1 RCIM_Cubic>,<at 1 0 RCIM_Cubic>]}}}, Scale Curve=1.5)
    Rotation Vector=&RandomRangeVector(Minimum=<-1 -1 -1>, Maximum=<1 1 1>)
   )
   &SpriteRotationRate(Rotation Rate=50)
   &ShapeLocation(
    MaskSpawnGroup=false
    Rotation Coordinate Space=::World
    Rotation Mode=::NewEnumerator2
    Sphere Distribution=::NewEnumerator1
    Sphere Radius=32
    Sphere Radius Position:NiagaraFloat=&RandomRangeFloat(Minimum=0.9, Maximum=1)
    U Position=&RandomRangeFloat(Minimum=0, Maximum=1)
    UseOwnerScale=true
    V Position=&RandomRangeFloat(Minimum=0, Maximum=1)
    Yaw / Pitch / Roll=&VectorFromCurve(
     VectorCurve=DataInterface() {propertyValues {
       XCurve {keys [<at 0 0.2 RCIM_Cubic>,<at 1 0 RCIM_Cubic>]}
       YCurve {keys [<at 0.0043 0.2>,<at 1 0>]}
       ZCurve {keys [<at 0 0.6 RCIM_Cubic>,<at 1 0 RCIM_Cubic>]}
      }}
     Scale Curve=&RandomRangeVector(Minimum=<-1 -1 -1>, Maximum=<2 2 2>)
    )
   )
   &GravityForce(Coordinate Space=::Local, Gravity=&RandomRangeVector(Minimum=<0 -500 -50>, Maximum=<0 -750 -110>))
   &Drag(
    Drag=&MergeHelper_FloatSwitch(0=35, 1=30, 2=?, 3=?, Index=&SpawnGroup(), Num=::NewEnumerator0)
    Use Linear Drag=true
    Use Rotational Drag=true
   )
   &SolveForcesAndVelocity(Clamp Velocity=false, Limit Acceleration=false)
   &AlignSpriteToMeshOrientation(SetAlignment=true, SetFacing=true)
   &ScaleColor(
    Scale Alpha=&FloatFromCurve(
     FloatCurve=DataInterface() {propertyValues {Curve {keys [
         {interpMode=RCIM_Cubic, tangentMode=RCTM_User, time=0, value=0}
         {arriveTangentWeight=0.0167, interpMode=RCIM_Cubic, leaveTangentWeight=0.1185, tangentMode=RCTM_User, time=0.1436, value=1}
         {arriveTangentWeight=0.2459, interpMode=RCIM_Cubic, leaveTangentWeight=0.1345, tangentMode=RCTM_User, time=0.65, value=1}
         {tangentMode=RCTM_User, time=1.0005, value=0}
        ]}}}
     Scale Curve=&RandomRangeFloat(Minimum=0.6, Maximum=1)
    )
    Scale RGB=&Multiply_VectorByFloat(Vector=<1 1 1>, Float=1)
    ScaleA=true
    ScaleRGB=true
   )
  ]
  rendererClasses [&NiagaraSpriteRendererProperties]
  renderers [
   {rendererClass=&NiagaraSpriteRendererProperties}
   {rendererClass=&NiagaraSpriteRendererProperties, rendererIndex=1}
  ]
 }
 Snowflakes_SL_Low {
  emitterUpdateScript [
   &EmitterState(Life Cycle Mode=::NewEnumerator1, Loop Behavior=::NewEnumerator1, Loop Delay=0.5, Loop Duration=9, UseLoopDelay=true)
   &SpawnBurst_Instantaneous(Spawn Count=75, Use Spawn Probability=false, UseLoopCountLimit=false)
   &SpawnBurst_Instantaneous(Spawn Count=70, Spawn Group=1, Spawn Time=6.2, Use Spawn Probability=false, UseLoopCountLimit=false) {module=SpawnBurst_Instantaneous001}
  ]
  particleSpawnScript [
   &InitializeParticle(
    Color=&FNTimeOfDayColor(
     Night Color=&RandomRangeLinearColor(Minimum=<rgba 0.3449 0.6586 1 0.8>, Maximum=<rgba 0.4305 0.5338 0.6354 0.8>)
     Morning Color=&RandomRangeLinearColor(Minimum=<rgba 0.3449 0.6586 1 0.8>, Maximum=<rgba 0.4305 0.5338 0.6354 0.8>)
     Day Color=&RandomRangeLinearColor(Minimum=<rgba 0.4138 0.7904 1.2 1>, Maximum=<rgba 0.6775 0.8401 1 1>)
     Evening Color=&RandomRangeLinearColor(Minimum=<rgba 0.3449 0.6586 1 0.8>, Maximum=<rgba 0.4305 0.5338 0.6354 0.8>)
    )
    Lifetime=&MergeHelper_FloatSwitch(
     0=&RandomRangeFloat(Minimum=5.5, Maximum=6.5)
     1=&RandomRangeFloat(Minimum=1.8, Maximum=2.2)
     2=?
     3=?
     Index=&SpawnGroup()
     Num=::NewEnumerator0
    )
    Mesh Scale Mode=::NewEnumerator1
    Mesh Uniform Scale Max=0.06
    Mesh Uniform Scale Min=0.015
    Sprite Size Mode=::NewEnumerator1
    Uniform Sprite Size Max=9
    Uniform Sprite Size Min=1
    UsePositionOffset=true
    Write Material Random=true
   )
   &AddVelocity(Velocity=<0 0 0>)
   /SetVariables(Particles.VisibilityTag=&SpawnGroup()) {isSetParameters=true, module=SetVariables_#1}
  ]
  particleUpdateScript [
   &ParticleState()
   &UpdateMeshOrientation(
    Rotation Rate=&FloatFromCurve(FloatCurve=DataInterface() {propertyValues {Curve {keys [<at 0 1 RCIM_Cubic>,<at 1 0 RCIM_Cubic>]}}}, Scale Curve=1.5)
    Rotation Vector=&RandomRangeVector(Minimum=<-1 -1 -1>, Maximum=<1 1 1>)
   )
   &SpriteRotationRate(Rotation Rate=50)
   &ShapeLocation(
    MaskSpawnGroup=false
    Rotation Coordinate Space=::World
    Rotation Mode=::NewEnumerator2
    Sphere Distribution=::NewEnumerator1
    Sphere Radius=32
    Sphere Radius Position:NiagaraFloat=&RandomRangeFloat(Minimum=0.9, Maximum=1)
    U Position=&RandomRangeFloat(Minimum=0, Maximum=1)
    UseOwnerScale=true
    V Position=&RandomRangeFloat(Minimum=0, Maximum=1)
    Yaw / Pitch / Roll=&VectorFromCurve(
     VectorCurve=DataInterface() {propertyValues {
       XCurve {keys [<at 0 0.2 RCIM_Cubic>,<at 1 0 RCIM_Cubic>]}
       YCurve {keys [<at 0.0043 0.2>,<at 1 0>]}
       ZCurve {keys [<at 0 0.6 RCIM_Cubic>,<at 1 0 RCIM_Cubic>]}
      }}
     Scale Curve=&RandomRangeVector(Minimum=<-1 -1 -1>, Maximum=<2 2 2>)
    )
   )
   &GravityForce(Coordinate Space=::Local, Gravity=&RandomRangeVector(Minimum=<0 -500 -50>, Maximum=<0 -750 -110>))
   &Drag(
    Drag=&MergeHelper_FloatSwitch(0=35, 1=30, 2=?, 3=?, Index=&SpawnGroup(), Num=::NewEnumerator0)
    Use Linear Drag=true
    Use Rotational Drag=true
   )
   &SolveForcesAndVelocity(Clamp Velocity=false, Limit Acceleration=false)
   &AlignSpriteToMeshOrientation(SetAlignment=true, SetFacing=true)
   &ScaleColor(
    Scale Alpha=&FloatFromCurve(
     FloatCurve=DataInterface() {propertyValues {Curve {keys [
         {interpMode=RCIM_Cubic, tangentMode=RCTM_User, time=0, value=0}
         {arriveTangentWeight=0.0167, interpMode=RCIM_Cubic, leaveTangentWeight=0.1185, tangentMode=RCTM_User, time=0.1436, value=1}
         {arriveTangentWeight=0.2459, interpMode=RCIM_Cubic, leaveTangentWeight=0.1345, tangentMode=RCTM_User, time=0.65, value=1}
         {tangentMode=RCTM_User, time=1.0005, value=0}
        ]}}}
     Scale Curve=&RandomRangeFloat(Minimum=0.6, Maximum=1)
    )
    Scale RGB=&Multiply_VectorByFloat(Vector=<1 1 1>, Float=1)
    ScaleA=true
    ScaleRGB=true
   )
  ]
  rendererClasses [&NiagaraSpriteRendererProperties]
  renderers [
   {rendererClass=&NiagaraSpriteRendererProperties}
   {rendererClass=&NiagaraSpriteRendererProperties, rendererIndex=1}
  ]
 }
 Snowflakes_Vortex_GPU {
  simTarget=GPUComputeSim
  emitterUpdateScript [
   &EmitterState(Life Cycle Mode=::NewEnumerator1, Loop Behavior=::NewEnumerator1, Loop Delay=0.5, Loop Duration=9, UseLoopDelay=true)
   &SpawnBurst_Instantaneous(Spawn Count=120, Use Spawn Probability=false, UseLoopCountLimit=false)
   &SpawnBurst_Instantaneous(Spawn Count=120, Spawn Group=1, Spawn Time=6.2, Use Spawn Probability=false, UseLoopCountLimit=false) {module=SpawnBurst_Instantaneous001}
  ]
  particleSpawnScript [
   &InitializeParticle(
    Color=&FNTimeOfDayColor(
     Night Color=&RandomRangeLinearColor(Minimum=<rgba 0.3449 0.6586 1 0.8>, Maximum=<rgba 0.4305 0.5338 0.6354 0.8>)
     Morning Color=&RandomRangeLinearColor(Minimum=<rgba 0.3449 0.6586 1 0.8>, Maximum=<rgba 0.4305 0.5338 0.6354 0.8>)
     Day Color=&RandomRangeLinearColor(Minimum=<rgba 0.4138 0.7904 1.2 1>, Maximum=<rgba 0.6775 0.8401 1 1>)
     Evening Color=&RandomRangeLinearColor(Minimum=<rgba 0.3449 0.6586 1 0.8>, Maximum=<rgba 0.4305 0.5338 0.6354 0.8>)
    )
    Lifetime=&MergeHelper_FloatSwitch(
     0=&RandomRangeFloat(Minimum=6.5, Maximum=7)
     1=&RandomRangeFloat(Minimum=4.2, Maximum=5)
     2=?
     3=?
     Index=&SpawnGroup()
     Num=::NewEnumerator0
    )
    Mesh Scale Mode=::NewEnumerator1
    Mesh Uniform Scale Max=0.06
    Mesh Uniform Scale Min=0.015
    Sprite Size Mode=::NewEnumerator1
    Uniform Sprite Size Max=8
    Uniform Sprite Size Min=1
    UsePositionOffset=true
    Write Material Random=true
   )
   &SphereLocation(
    MaskSpawnGroup=false
    Offset=<0 0 -5>
    Override Seed=false
    Sphere Orientation Axis:Vector3f=<0 0 1>
    Sphere Radius=25
    Surface Only=false
    UseNonUniformScale=false
    UseOffset=true
    UseSphereOrientation=true
   )
   &AddVelocity(
    Cone Angle=90
    Cone Axis=<0 0 1>
    Module.UseConeFalloff=false
    Module.UseVelDistribution=false
    Override Seed=false
    Velocity Mode=::NewEnumerator2
    Velocity Speed=&RandomRangeFloat(Minimum=250, Maximum=300)
   )
   /SetVariables(Particles.VisibilityTag=&SpawnGroup()) {isSetParameters=true, module=SetVariables_#1}
   &CameraOffset(Camera Offset Amount=-5)
  ]
  particleUpdateScript [
   &ParticleState()
   &SpriteRotationRate(Rotation Rate=50)
   &UpdateMeshOrientation(
    Rotation Rate=&FloatFromCurve(FloatCurve=DataInterface() {propertyValues {Curve {keys [<at 0 1 RCIM_Cubic>,<at 1 0 RCIM_Cubic>]}}}, Scale Curve=1.5)
    Rotation Vector=&RandomRangeVector(Minimum=<-1 -1 -1>, Maximum=<1 1 1>)
   )
   &KillParticlesInVolume(Invert Volume=true, Sphere Radius=32)
   &VortexForce(
    Origin Pull Amount:NiagaraFloat=&MergeHelper_FloatSwitch(0=300, 1=450, 2=?, 3=?, Index=&SpawnGroup(), Num=::NewEnumerator0)
    UseOriginOffset=true
    Vortex Axis=<0 0 0>
    Vortex Axis Coordinate Space:ENiagaraCoordinateSpace=::Local
    Vortex Force Amount=&MergeHelper_FloatSwitch(
     0=&FloatFromCurve(
      FloatCurve=DataInterface() {propertyValues {Curve {keys [<at 0.2935 1 RCIM_Cubic>,<at 1 0.36 RCIM_Cubic>]}}}
      Scale Curve=&SelectFloatFromArray(
       Array Sampling Mode=::NewEnumerator0
       Direct Array Index=?
       Float Selection Array=DataInterface() {propertyValues {FloatData [2750,4000,-2750,-4000]}}
       Random Seed=0
      )
     )
     1=&FloatFromCurve(
      FloatCurve=DataInterface() {propertyValues {Curve {keys [<at 0.3 1 RCIM_Cubic>,<at 1 0.2 RCIM_Cubic>]}}}
      Scale Curve=&RandomRangeFloat(Minimum=-3500, Maximum=-4500)
     )
     2=?
     3=?
     Index=&SpawnGroup()
     Num=::NewEnumerator0
    )
    Vortex Origin Offset:Vector3f=<-5 0 0>
   )
   &GravityForce(Coordinate Space=::Local, Gravity=&RandomRangeVector(Minimum=<0 -500 -50>, Maximum=<0 -750 -110>))
   &Drag(
    Drag=&MergeHelper_FloatSwitch(0=35, 1=30, 2=?, 3=?, Index=&SpawnGroup(), Num=::NewEnumerator0)
    Use Linear Drag=true
    Use Rotational Drag=true
   )
   &SolveForcesAndVelocity(Clamp Velocity=false, Limit Acceleration=false)
   &AlignSpriteToMeshOrientation(SetAlignment=true, SetFacing=true)
   &ScaleColor(
    Scale Alpha=&FloatFromCurve(
     FloatCurve=DataInterface() {propertyValues {Curve {keys [
         {interpMode=RCIM_Cubic, tangentMode=RCTM_User, time=0, value=0}
         {arriveTangentWeight=0.0167, interpMode=RCIM_Cubic, leaveTangentWeight=0.1185, tangentMode=RCTM_User, time=0.1436, value=1}
         {arriveTangentWeight=0.2459, interpMode=RCIM_Cubic, leaveTangentWeight=0.1345, tangentMode=RCTM_User, time=0.65, value=1}
         {tangentMode=RCTM_User, time=1.0005, value=0}
        ]}}}
     Scale Curve=&RandomRangeFloat(Minimum=0.6, Maximum=1)
    )
    Scale RGB=&Multiply_VectorByFloat(Vector=<1 1 1>, Float=1)
    ScaleA=true
    ScaleRGB=true
   )
  ]
  rendererClasses [&NiagaraSpriteRendererProperties]
  renderers [
   {rendererClass=&NiagaraSpriteRendererProperties}
   {rendererClass=&NiagaraSpriteRendererProperties, rendererIndex=1}
  ]
 }
 Snowflakes_Vortex_Low {
  emitterUpdateScript [
   &EmitterState(Life Cycle Mode=::NewEnumerator1, Loop Behavior=::NewEnumerator1, Loop Delay=0.5, Loop Duration=9, UseLoopDelay=true)
   &SpawnBurst_Instantaneous(Spawn Count=120, Use Spawn Probability=false, UseLoopCountLimit=false)
   &SpawnBurst_Instantaneous(Spawn Count=120, Spawn Group=1, Spawn Time=6.2, Use Spawn Probability=false, UseLoopCountLimit=false) {module=SpawnBurst_Instantaneous001}
  ]
  particleSpawnScript [
   &InitializeParticle(
    Color=&FNTimeOfDayColor(
     Night Color=&RandomRangeLinearColor(Minimum=<rgba 0.3449 0.6586 1 0.8>, Maximum=<rgba 0.4305 0.5338 0.6354 0.8>)
     Morning Color=&RandomRangeLinearColor(Minimum=<rgba 0.3449 0.6586 1 0.8>, Maximum=<rgba 0.4305 0.5338 0.6354 0.8>)
     Day Color=&RandomRangeLinearColor(Minimum=<rgba 0.4138 0.7904 1.2 1>, Maximum=<rgba 0.6775 0.8401 1 1>)
     Evening Color=&RandomRangeLinearColor(Minimum=<rgba 0.3449 0.6586 1 0.8>, Maximum=<rgba 0.4305 0.5338 0.6354 0.8>)
    )
    Lifetime=&MergeHelper_FloatSwitch(
     0=&RandomRangeFloat(Minimum=6.5, Maximum=7)
     1=&RandomRangeFloat(Minimum=4.2, Maximum=5)
     2=?
     3=?
     Index=&SpawnGroup()
     Num=::NewEnumerator0
    )
    Mesh Scale Mode=::NewEnumerator1
    Mesh Uniform Scale Max=0.06
    Mesh Uniform Scale Min=0.015
    Sprite Size Mode=::NewEnumerator1
    Uniform Sprite Size Max=8
    Uniform Sprite Size Min=1
    UsePositionOffset=true
    Write Material Random=true
   )
   &SphereLocation(
    MaskSpawnGroup=false
    Offset=<0 0 -5>
    Override Seed=false
    Sphere Orientation Axis:Vector3f=<0 0 1>
    Sphere Radius=25
    Surface Only=false
    UseNonUniformScale=false
    UseOffset=true
    UseSphereOrientation=true
   )
   &AddVelocity(
    Cone Angle=90
    Cone Axis=<0 0 1>
    Module.UseConeFalloff=false
    Module.UseVelDistribution=false
    Override Seed=false
    Velocity Mode=::NewEnumerator2
    Velocity Speed=&RandomRangeFloat(Minimum=250, Maximum=300)
   )
   /SetVariables(Particles.VisibilityTag=&SpawnGroup()) {isSetParameters=true, module=SetVariables_#1}
  ]
  particleUpdateScript [
   &ParticleState()
   &SpriteRotationRate(Rotation Rate=50)
   &UpdateMeshOrientation(
    Rotation Rate=&FloatFromCurve(FloatCurve=DataInterface() {propertyValues {Curve {keys [<at 0 1 RCIM_Cubic>,<at 1 0 RCIM_Cubic>]}}}, Scale Curve=1.5)
    Rotation Vector=&RandomRangeVector(Minimum=<-1 -1 -1>, Maximum=<1 1 1>)
   )
   &KillParticlesInVolume(Invert Volume=true, Sphere Radius=32)
   &VortexForce(
    Origin Pull Amount:NiagaraFloat=&MergeHelper_FloatSwitch(0=300, 1=450, 2=?, 3=?, Index=&SpawnGroup(), Num=::NewEnumerator0)
    UseOriginOffset=true
    Vortex Axis=<0 0 0>
    Vortex Axis Coordinate Space:ENiagaraCoordinateSpace=::Local
    Vortex Force Amount=&MergeHelper_FloatSwitch(
     0=&FloatFromCurve(
      FloatCurve=DataInterface() {propertyValues {Curve {keys [<at 0.2935 1 RCIM_Cubic>,<at 1 0.36 RCIM_Cubic>]}}}
      Scale Curve=&SelectFloatFromArray(
       Array Sampling Mode=::NewEnumerator0
       Direct Array Index=?
       Float Selection Array=DataInterface() {propertyValues {FloatData [2750,4000,-2750,-4000]}}
       Random Seed=0
      )
     )
     1=&FloatFromCurve(
      FloatCurve=DataInterface() {propertyValues {Curve {keys [<at 0.3 1 RCIM_Cubic>,<at 1 0.2 RCIM_Cubic>]}}}
      Scale Curve=&RandomRangeFloat(Minimum=-3500, Maximum=-4500)
     )
     2=?
     3=?
     Index=&SpawnGroup()
     Num=::NewEnumerator0
    )
    Vortex Origin Offset:Vector3f=<-5 0 0>
   )
   &GravityForce(Coordinate Space=::Local, Gravity=&RandomRangeVector(Minimum=<0 -500 -50>, Maximum=<0 -750 -110>))
   &Drag(
    Drag=&MergeHelper_FloatSwitch(0=35, 1=30, 2=?, 3=?, Index=&SpawnGroup(), Num=::NewEnumerator0)
    Use Linear Drag=true
    Use Rotational Drag=true
   )
   &SolveForcesAndVelocity(Clamp Velocity=false, Limit Acceleration=false)
   &AlignSpriteToMeshOrientation(SetAlignment=true, SetFacing=true)
   &ScaleColor(
    Scale Alpha=&FloatFromCurve(
     FloatCurve=DataInterface() {propertyValues {Curve {keys [
         {interpMode=RCIM_Cubic, tangentMode=RCTM_User, time=0, value=0}
         {arriveTangentWeight=0.0167, interpMode=RCIM_Cubic, leaveTangentWeight=0.1185, tangentMode=RCTM_User, time=0.1436, value=1}
         {arriveTangentWeight=0.2459, interpMode=RCIM_Cubic, leaveTangentWeight=0.1345, tangentMode=RCTM_User, time=0.65, value=1}
         {tangentMode=RCTM_User, time=1.0005, value=0}
        ]}}}
     Scale Curve=&RandomRangeFloat(Minimum=0.6, Maximum=1)
    )
    Scale RGB=&Multiply_VectorByFloat(Vector=<1 1 1>, Float=1)
    ScaleA=true
    ScaleRGB=true
   )
  ]
  rendererClasses [&NiagaraSpriteRendererProperties]
  renderers [
   {rendererClass=&NiagaraSpriteRendererProperties}
   {rendererClass=&NiagaraSpriteRendererProperties, rendererIndex=1}
  ]
 }
 WaterSwishes001 {
  emitterUpdateScript [
   &EmitterState(Life Cycle Mode=::NewEnumerator1, Loop Behavior=::NewEnumerator1, Loop Delay=6.8, Loop Duration=1, UseLoopDelay=true)
   &SpawnRate(SpawnRate=55, Use Spawn Probability=false)
  ]
  particleSpawnScript [
   &InitializeParticle(
    Color Maximum=<rgba 0.2734 0.4056 0.5 0.3>
    Color Minimum=<rgba 0.7917 1.4361 2 0.2>
    Color Mode=::NewEnumerator2
    Lifetime Max=0.5
    Lifetime Min=0.25
    Lifetime Mode=::NewEnumerator1
    Mesh Scale=&Multiply_VectorByFloat(
     Vector=&RandomRangeVector(Minimum=<1.2 1.2 2>, Maximum=<1.3 1.3 2.7>)
     Float=&RandomRangeFloat(Minimum=0.208, Maximum=0.39)
    )
    Mesh Scale Mode=::NewEnumerator4
    UsePositionOffset=false
    Write Material Random=true
   )
   &InitialMeshOrientation(
    Mesh Axis To Orient=<0 0 0>
    Mesh Orientation Mode=::NewEnumerator2
    Orientation Coordinate Space=::Local
    Orientation Vector=<0 1 0>
    Rotation=&Multiply_VectorByFloat(
     Vector=&RandomRangeVector(Minimum=<0 0 0>, Maximum=<0.1 0.1 1>)
     Float=&MakeFloatFromInt(Int=&SelectIntFromArray(
       Array Sampling Mode=::NewEnumerator1
       Direct Array Index=?
       Int Selection Array=DataInterface() {propertyValues {IntData [-1,1]}}
       Random Seed=0
      ))
    )
    Use Rotation Vector=true
   )
   &ShapeLocation(
    Cylinder Height=0.5
    Cylinder Radius=0.5
    MaskSpawnGroup=false
    Non Uniform Scale=<1 12 1>
    Override Seed=false
    Shape Primitive=::NewEnumerator1
    Surface Only=false
    UseOwnerScale=true
   )
  ]
  particleUpdateScript [
   &ParticleState()
   &DynamicMaterialParameters(
    Index 0 Param 1=&FloatFromCurve(FloatCurve=DataInterface() {propertyValues {Curve {keys [
         {interpMode=RCIM_Cubic, tangentMode=RCTM_User, time=0, value=1}
         <at 0.7 0.35>
         {interpMode=RCIM_Cubic, tangentMode=RCTM_User, time=1, value=0}
        ]}}})
    Module.Param0WriteEnabled=true
    Module.Param1WriteEnabled=true
    Module.Param2WriteEnabled=true
    Module.Param3WriteEnabled=true
   )
   &ScaleMeshSize(Scale Factor=&Multiply_VectorByFloat(
     Vector=&Multiply_VectorByFloat(Vector=$Engine.Owner.Scale, Float=1)
     Float=&FloatFromCurve(
      FloatCurve=DataInterface() {propertyValues {Curve {keys [
          {arriveTangent=2.0131, interpMode=RCIM_Cubic, leaveTangent=2.0131, tangentMode=RCTM_User, time=0, value=0}
          {arriveTangent=0.0166, interpMode=RCIM_Cubic, leaveTangent=0.0166, tangentMode=RCTM_User, time=0.7, value=1}
         ]}}}
      Scale Curve=&FN_FrontEndFloatModifier(Gameplay=1, Front End=1)
     )
    ))
   &UpdateMeshOrientation(Rotation Rate=&RandomRangeFloat(Minimum=-0.25, Maximum=-0.5), Rotation Vector=<0 0 1>)
   &ScaleColor(
    Scale Alpha=&FloatFromCurve(FloatCurve=DataInterface() {propertyValues {Curve {keys [
         {arriveTangent=4.4696, interpMode=RCIM_Cubic, leaveTangent=4.4696, tangentMode=RCTM_User, time=0, value=0}
         {interpMode=RCIM_Cubic, tangentMode=RCTM_User, time=0.5, value=1}
         {arriveTangent=-5.2232, tangentMode=RCTM_User, time=1, value=0}
        ]}}})
    ScaleA=true
    ScaleRGB=true
   )
   &GravityForce(Coordinate Space=::Local, Gravity=&RandomRangeVector(Minimum=<0 -5 0>, Maximum=<1 -10 1>))
   &SolveForcesAndVelocity(Clamp Velocity=false, Limit Acceleration=false)
  ]
  rendererClasses [&NiagaraMeshRendererProperties]
  renderers [{rendererClass=&NiagaraMeshRendererProperties, rendererIndex=0}]
 }
 WaterSwishes_1 {
  emitterUpdateScript [
   &EmitterState(Life Cycle Mode=::NewEnumerator1, Loop Behavior=::NewEnumerator1, Loop Duration=1.5, UseLoopDelay=true)
   &SpawnRate(SpawnRate=40, Use Spawn Probability=false)
  ]
  particleSpawnScript [
   &InitializeParticle(
    Color Maximum=<rgba 0.2734 0.4056 0.5 0.3>
    Color Minimum=<rgba 0.7917 1.4361 2 0.2>
    Color Mode=::NewEnumerator2
    Lifetime Max=0.5
    Lifetime Min=0.25
    Lifetime Mode=::NewEnumerator1
    Mesh Scale=&Multiply_VectorByFloat(
     Vector=&RandomRangeVector(Minimum=<1.2 1.2 2>, Maximum=<1.3 1.3 2.7>)
     Float=&RandomRangeFloat(Minimum=0.208, Maximum=0.39)
    )
    Mesh Scale Mode=::NewEnumerator4
    UsePositionOffset=false
    Write Material Random=true
   )
   &InitialMeshOrientation(
    Mesh Axis To Orient=<0 0 0>
    Mesh Orientation Mode=::NewEnumerator2
    Orientation Coordinate Space=::Local
    Orientation Vector=<0 1 0>
    Rotation=&Multiply_VectorByFloat(
     Vector=&RandomRangeVector(Minimum=<0 0 0>, Maximum=<0.1 0.1 1>)
     Float=&MakeFloatFromInt(Int=&SelectIntFromArray(
       Array Sampling Mode=::NewEnumerator1
       Direct Array Index=?
       Int Selection Array=DataInterface() {propertyValues {IntData [-1,1]}}
       Random Seed=0
      ))
    )
    Use Rotation Vector=true
   )
   &ShapeLocation(
    Cylinder Height=0.5
    Cylinder Radius=0.5
    MaskSpawnGroup=false
    Non Uniform Scale=<1 12 1>
    Override Seed=false
    Shape Primitive=::NewEnumerator1
    Surface Only=false
    UseOwnerScale=true
   )
   /SetVariables(Particles.VisibilityTag=&SpawnGroup()) {enabled=false, isSetParameters=true}
  ]
  particleUpdateScript [
   &ParticleState()
   &DynamicMaterialParameters(
    Index 0 Param 1=&FloatFromCurve(FloatCurve=DataInterface() {propertyValues {Curve {keys [
         {interpMode=RCIM_Cubic, tangentMode=RCTM_User, time=0, value=1}
         <at 0.7 0.35>
         {interpMode=RCIM_Cubic, tangentMode=RCTM_User, time=1, value=0}
        ]}}})
    Module.Param0WriteEnabled=true
    Module.Param1WriteEnabled=true
    Module.Param2WriteEnabled=true
    Module.Param3WriteEnabled=true
   )
   &ScaleMeshSize(Scale Factor=&Multiply_VectorByFloat(
     Vector=&Multiply_VectorByFloat(Vector=$Engine.Owner.Scale, Float=1)
     Float=&FloatFromCurve(
      FloatCurve=DataInterface() {propertyValues {Curve {keys [
          {arriveTangent=2.0131, interpMode=RCIM_Cubic, leaveTangent=2.0131, tangentMode=RCTM_User, time=0, value=0}
          {arriveTangent=0.0166, interpMode=RCIM_Cubic, leaveTangent=0.0166, tangentMode=RCTM_User, time=0.7, value=1}
         ]}}}
      Scale Curve=&FN_FrontEndFloatModifier(Gameplay=1, Front End=1)
     )
    ))
   &UpdateMeshOrientation(Rotation Rate=&RandomRangeFloat(Minimum=-0.25, Maximum=-0.5), Rotation Vector=<0 0 -1>)
   &ScaleColor(
    Scale Alpha=&FloatFromCurve(FloatCurve=DataInterface() {propertyValues {Curve {keys [
         {arriveTangent=4.4696, interpMode=RCIM_Cubic, leaveTangent=4.4696, tangentMode=RCTM_User, time=0, value=0}
         {interpMode=RCIM_Cubic, tangentMode=RCTM_User, time=0.5, value=1}
         {arriveTangent=-5.2232, tangentMode=RCTM_User, time=1, value=0}
        ]}}})
    ScaleA=true
    ScaleRGB=true
   )
   &GravityForce(Coordinate Space=::Local, Gravity=&RandomRangeVector(Minimum=<0 -5 0>, Maximum=<1 -10 1>))
   &SolveForcesAndVelocity(Clamp Velocity=false, Limit Acceleration=false)
  ]
  rendererClasses [&NiagaraMeshRendererProperties]
  renderers [{rendererClass=&NiagaraMeshRendererProperties, rendererIndex=0}]
 }
}
summary {
 systemName=NS_Emote_IcyScarf_SnowFlakes
 emitters [
  {emitterName=Snowflakes_Vortex_GPU, simTarget=GPUComputeSim, rendererClasses [&NiagaraSpriteRendererProperties]}
  {emitterName=WaterSwishes_1, rendererClasses [&NiagaraMeshRendererProperties]}
  {emitterName=WaterSwishes001, rendererClasses [&NiagaraMeshRendererProperties]}
  {emitterName=Snowflakes_SL_GPU, simTarget=GPUComputeSim, rendererClasses [&NiagaraSpriteRendererProperties]}
  {emitterName=Snowflakes_Vortex_Low, rendererClasses [&NiagaraSpriteRendererProperties]}
  {emitterName=Snowflakes_SL_Low, rendererClasses [&NiagaraSpriteRendererProperties]}
 ]
 userVariables [
  User.bIsFrontEnd {type=NiagaraBool}
  User.bIsFrontEndPreview {type=NiagaraBool}
  User.Radius {type=NiagaraFloat, defaultValue=100, description=""}
 ]
}
